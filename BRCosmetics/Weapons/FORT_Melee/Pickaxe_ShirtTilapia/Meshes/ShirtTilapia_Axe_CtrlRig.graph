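# Control Rig for the ShirtTilapia pickaxe that procedurally animates its eyeball bone with a wandering, springy motion.
# Two seeded random float generators periodically pick new Y and Z Euler targets, while X comes from the bone's current local rotation.
# A vector spring interpolator smooths toward that target with a constant Y/Z force.
# The result converts back to a quaternion and drives the eyeball through Modify Transforms.
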ControlRigBlueprint /BRCosmetics/Weapons/FORT_Melee/Pickaxe_ShirtTilapia/Meshes/ShirtTilapia_Axe_CtrlRig
graphs [
 RigVMFunctionLibrary {links [], nodes []}
 RigVMModel {nodes [
   RigUnit_BeginExecution = &RigUnit_BeginExecution()
   n1 = &RigVMFunction_MathQuaternionToEuler(
    RotationOrder=ZYX
    Value=&RigUnit_GetTransform(Item=(Type=Bone,Name="eyeball"), Space=LocalSpace, bInitial=False).Transform.Rotation
   )
   SpringInterp = /Script/ControlRig.RigUnit_SpringInterpVectorV2(
    CriticalDamping=3
    Force=(X=0,Y=1,Z=1)
    Strength=5
    Target=(X=0,Y=0,Z=0)
    TargetVelocityAmount=0
    bInitializeFromTarget=False
    bUseCurrentInput=False
    Target.Y=/Script/RigVM.RigVMFunction_RandomFloat(Duration=3, Maximum=30, Minimum=-30, Seed=217) {is=struct RigVMUnitNode, template="Random::Execute(in Seed,in Minimum,in Maximum,in Duration,out Result)"}.Result
    Target.Z=&RigVMFunction_RandomFloat(Duration=3, Maximum=30, Minimum=-30, Seed=641) {template="Random::Execute(in Seed,in Minimum,in Maximum,in Duration,out Result)"}.Result
    Current=@n1.Result
    Target.X=@n1.Result.X
   ) {
    is=struct RigVMUnitNode
    template="SpringInterp::Execute(in Target,in Strength,in CriticalDamping,in Force,in bUseCurrentInput,in Current,in TargetVelocityAmount,in bInitializeFromTarget,out Result,out Velocity)"
   }
   RigVMFunction_MathQuaternionFromEuler = &RigVMFunction_MathQuaternionFromEuler(Euler=@SpringInterp.Result)
   &RigUnit_ModifyTransforms(
    ItemToModify=((Item=(Type=Bone,Name="eyeball"),Transform=(Rotation=(X=0,Y=0,Z=0,W=1),Translation=(X=0,Y=0,Z=0),Scale3D=(X=1,Y=1,Z=1))))
    Weight=1
    WeightMinimum=0
    ExecutePin=@RigUnit_BeginExecution.ExecutePin
    ItemToModify.0.Transform.Rotation=@RigVMFunction_MathQuaternionFromEuler.Result
   )
  ]}
]
hierarchy [
 b_AR_Root {loc=<0 0 0>, rot=<0 0 0.7071 0.7071>}
 Melee_TwoHanded_Handle {parent=b_AR_Root, loc=<8.1477 -0.008 -29.0793>, rot=<0 0 0.7071 -0.7071>}
 eyeball_MSTR {parent=Melee_TwoHanded_Handle, loc=<0.8489 -0.9792 91.7356>, rot=<0 0 0 -1>}
 eyeball {parent=eyeball_MSTR, loc=<0 0 0>, rot=<0 0 0 1>}
]
