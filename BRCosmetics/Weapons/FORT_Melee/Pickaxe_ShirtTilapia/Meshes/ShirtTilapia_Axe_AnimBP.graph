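# Animation Blueprint for the ShirtTilapia pickaxe that post-processes its pose with a Control Rig.
# The AnimGraph passes the input pose through ShirtTilapia_Axe_CtrlRig to the output, with no variable inputs shown.
# The EventGraph holds real-time sine math that sets an oscillating EyeRotate pitch, plus an unused phase-shifted copy.
# BlueprintUpdateAnimation has nothing connected, so that eye logic appears not to run.
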AnimBlueprint /BRCosmetics/Weapons/FORT_Melee/Pickaxe_ShirtTilapia/Meshes/ShirtTilapia_Axe_AnimBP
target_skeleton=&/ShirtTilapia_Axe_Skeleton
graphs [
 AnimGraph {nodes [
   AnimGraphNode_SubInput_0 = Input Pose()
   n1 = Control Rig(Source=@AnimGraphNode_SubInput_0) {
    alphaBoolBlend {blendOption=Linear}
    controlRigAssetReference {blueprintRigClass=&/ShirtTilapia_Axe_CtrlRig.ShirtTilapia_Axe_CtrlRig_C}
   }
   Output Pose(Result=@n1) {layerGroup=None}
  ]}
 EventGraph {nodes [
   n1 = Sequence()
   n2 = GetRealTimeSeconds()
   n3 = float * float(A="Sin (Radians)"(A=float * float(A=@n2, B=4)), B=10) {is=title K2Node_PromotableOperator}
   > Set EyeRotate(EyeRotate:Rotator=MakeRotator(Pitch=float + float(A=@n3, B=-10))) -> (Output_Get:Rotator)
   Reroute Node(InputPin=@n1.then_1)
   TryGetPawnOwner()
   n4 = GetRealTimeSeconds()
   n5 = "Sin (Radians)"(A=float + float(A=float * float(A=@n4, B=4), B=0.5))
   MakeRotator(Pitch=float + float(A=float * float(A=@n5, B=10), B=-10))
   Event BlueprintUpdateAnimation()
  ]}
]
variables [
 EyeDirection {type=number}
 EyeRotate {type=Rotator}
 Phase {type=integer}
 Timer {type=number}
]
